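# Niagara system for the Plumgold vehicle boost, a gold smoke trail with spark colour and size parameters, emitted from the car's boost sockets.
# One GoldSmoke emitter partitions particles across the ports, pushes them backward, and fades their inherited vehicle movement over life.
# SubUV animation, size and colour curves shape the sprites, drawn by three renderers.
# GameModeID switches retune spawn rate, lifetime, velocity, colour and curves per game mode.
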NiagaraSystem /VehicleCosmetics/Boosts/Boost_Plumgold/NS_Boost_Plumgold
dependencies {
 usedDataInterfaces [
  &NiagaraDataInterfaceCurve
  &NiagaraDataInterfaceSpriteRendererInfo
  &NiagaraDataInterfaceSkeletalMesh
  &NiagaraDataInterfaceArrayFloat3
 ]
 usedDynamicInputs [&SelectVectorFromArray]
 usedModules [
  &self:SystemSpawnScript.SystemScriptSource.SystemScriptGraph.NiagaraNodeAssignment_0.SetVariables_#1
  &self:SystemSpawnScript.SystemScriptSource.SystemScriptGraph.NiagaraNodeAssignment_1.SetVariables_#2
  &self:SystemSpawnScript.SystemScriptSource.SystemScriptGraph.NiagaraNodeAssignment_2.SetVariables_#3
  &SystemState
 ]
}
emitters {GoldSmoke {
  emitterSpawnScript [/SetVariables(Emitter.ContrastSparks:NiagaraFloat=&MergeHelper_FloatSwitch(0=4, 1=2, 2=2, 3=1.5, Index=$System.GameModeID, Num=::NewEnumerator2)) {isSetParameters=true}]
  emitterUpdateScript [
   &EmitterState()
   &VC_FilteredSocketCount()
   &SpawnRate(
    SpawnRate=&Multiply_Float(
     A=200
     B=/VehicleCosmetics/FX/Niagara/Modules/VC_Mask_Floats(Coro Value:NiagaraFloat=1.25, FrontEnd Value:NiagaraFloat=1.5, FN Value:NiagaraFloat=0.5)
    )
    Use Spawn Probability=false
   ) {enabled=false}
   &SpawnRate(
    SpawnRate=&MergeHelper_FloatSwitch(0=300, 1=400, 2=400, 3=250, Index=$System.GameModeID, Num=::NewEnumerator2)
    Use Spawn Probability=false
   ) {module=SpawnRate002}
  ]
  particleSpawnScript [
   &VC_PaintedVariantsOverride() {enabled=false}
   &InitializeParticle(
    Color=&MergeHelper_ColorSwitch(
     0=<rgba 6 2.8974 0.25 1>
     1=<rgba 6 2.8974 0.25 1>
     2=<rgba 6 2.8974 0.25 1>
     3=<rgba 8 3.8632 0.3333 1>
     Index=$System.GameModeID
     Num=::NewEnumerator2
    )
    Lifetime=&MergeHelper_FloatSwitch(
     0=&RandomRangeFloat(Minimum=0.35, Maximum=0.6)
     1=&RandomRangeFloat(Minimum=0.35, Maximum=0.6)
     2=&RandomRangeFloat(Minimum=0.8, Maximum=1.2)
     3=&RandomRangeFloat(Minimum=0.8, Maximum=1.2)
     Index=$System.GameModeID
     Num=::NewEnumerator2
    )
    Sprite Rotation Mode=::NewEnumerator2
    Sprite Size Mode=::NewEnumerator3
    Uniform Sprite Size=&Multiply_Float(A=&RandomRangeFloat(Minimum=80, Maximum=40), B=1)
    UsePositionOffset=true
    Write Material Random=true
   )
   &PartitionParticles(Number Of Partitions=$Emitter.PortCount)
   &SkeletalMeshLocation(
    Bone / Socket Index=$Particles.PartitionParticles.Partition
    Bone / Socket Sampling Mode=::NewEnumerator5
    MaskSpawnGroup=false
    Mesh Sampling Type=::NewEnumerator6
    Sampled Position Offset=&AddVectorToPosition(
     Position=&MergeHelper_VectorSwitch(0=<-20 0 0>, 1=<20 0 0>, 2=<0 0 0>, 3=<100 0 0>, Index=$System.GameModeID, Num=::NewEnumerator2)
     Vector=$User.DTCarOffset
    )
    Skeletal Mesh=$System.SkeletalMesh
   ) {enabled=false}
   &SkeletalMeshLocation(
    Bone / Socket Index=$Particles.PartitionParticles.Partition
    Bone / Socket Sampling Mode=::NewEnumerator5
    MaskSpawnGroup=false
    Mesh Sampling Type=::NewEnumerator6
    Sampled Position Offset=&AddVectorToPosition(
     Position=&VC_Mask_Vectors(Coro Value:Vector3f=<100 0 0>, FrontEnd Value:Vector3f=<-20 0 0>, FN Value:Vector3f=<-1200 0 0>)
     Vector=$User.DTCarOffset
    )
    Skeletal Mesh=$System.SkeletalMesh
   ) {module=SkeletalMeshLocation001}
   &AddVelocity(Velocity=&VC_Mask_Vectors(Coro Value:Vector3f=<-1000 0 0>, FrontEnd Value:Vector3f=<-1500 0 0>, FN Value:Vector3f=<-1200 0 0>)) {enabled=false}
   &AddVelocity(Velocity=&MergeHelper_VectorSwitch(0=<-1500 0 0>, 1=<-1500 0 0>, 2=<-1000 0 0>, 3=<-1000 0 0>, Index=$System.GameModeID, Num=::NewEnumerator2)) {module=AddVelocity001}
   /VehicleCosmetics/FX/VC_FXParents/Modules/DLR_ProductPreviews(GameModeID:NiagaraInt32=$System.GameModeID, bPreviewVehicleSpeed:NiagaraBool=$System.bPreviewVehicleSpeed)
  ]
  particleUpdateScript [
   &ParticleState()
   &ScaleSpriteSize(
    Scale Sprite Size Mode=::NewEnumerator0
    Uniform Scale Factor=&MergeHelper_FloatSwitch(
     0=&FloatFromCurve(FloatCurve=DataInterface() {propertyValues {Curve {
         defaultValue=0
         keys [
          {interpMode=RCIM_Cubic, leaveTangent=13.1104, tangentMode=RCTM_Break, time=0, value=0.1}
          {arriveTangent=13.1319, interpMode=RCIM_Cubic, leaveTangent=0.6083, tangentMode=RCTM_Break, time=0.1, value=1.6}
          {arriveTangent=0.8365, tangentMode=RCTM_User, time=1, value=2.3}
         ]
        }}})
     1=&FloatFromCurve(FloatCurve=DataInterface() {propertyValues {Curve {
         defaultValue=0
         keys [
          {interpMode=RCIM_Cubic, leaveTangent=13.746, tangentMode=RCTM_Break, time=0, value=0.3}
          {arriveTangent=5.0974, interpMode=RCIM_Cubic, leaveTangent=0.6083, tangentMode=RCTM_Break, time=0.1, value=1.6}
          {arriveTangent=0.8365, tangentMode=RCTM_User, time=1, value=2.3}
         ]
        }}})
     2=&FloatFromCurve(FloatCurve=DataInterface() {propertyValues {Curve {
         defaultValue=0
         keys [
          {interpMode=RCIM_Cubic, leaveTangent=47.5901, tangentMode=RCTM_Break, time=0, value=0.3}
          {arriveTangent=5.0974, interpMode=RCIM_Cubic, leaveTangent=0.6083, tangentMode=RCTM_Break, time=0.05, value=1.8}
          {arriveTangent=0.8365, tangentMode=RCTM_User, time=1, value=2.3}
         ]
        }}})
     3=&FloatFromCurve(FloatCurve=DataInterface() {propertyValues {Curve {
         defaultValue=0
         keys [
          {interpMode=RCIM_Cubic, leaveTangent=5.2758, tangentMode=RCTM_Break, time=0, value=0.3}
          {arriveTangent=5.0974, interpMode=RCIM_Cubic, leaveTangent=0.6083, tangentMode=RCTM_Break, time=0.25, value=1.6}
          {arriveTangent=0.8365, tangentMode=RCTM_User, time=1, value=2.3}
         ]
        }}})
     Index=$System.GameModeID
     Num=::NewEnumerator2
    )
   )
   &InheritSourceMovement(
    Applied Position Scale=&VectorFromFloat(Value=&FloatFromCurve(FloatCurve=DataInterface() {propertyValues {Curve {
         defaultValue=1
         keys [
          {arriveTangent=-1.7885, interpMode=RCIM_Cubic, leaveTangent=-1.7885, tangentMode=RCTM_User, time=0.05, value=1}
          {interpMode=RCIM_Cubic, tangentMode=RCTM_User, time=1, value=0}
         ]
        }}}))
    ApplyForce=false
    ApplyPosition=true
    ApplyPositionFalloff=false
   )
   &Drag(Use Linear Drag=true, Use Rotational Drag=true)
   &SubUVAnimation(
    "Sprite Renderer "=DataInterface() {propertyValues {SpriteRenderer=&self:GoldSmoke_0.NiagaraSpriteRendererProperties_5}}
    Start Frame Range Override=12
    UseEndFrame=true
    UseStartFrame=true
   )
   &ScaleColor(
    Scale Alpha=&FloatFromCurve(FloatCurve=DataInterface() {propertyValues {Curve {defaultValue=1, keys [<at 0 1>, {arriveTangentWeight=0.2667, interpMode=RCIM_Cubic, tangentMode=RCTM_Break, time=0.8, value=1}, <at 1 0>]}}})
    Scale RGB=&VectorFromFloat(Value=&MergeHelper_FloatSwitch(
      0=&FloatFromCurve(
       FloatCurve=DataInterface() {propertyValues {Curve {keys [
           {interpMode=RCIM_Cubic, leaveTangentWeight=0.3333, tangentMode=RCTM_Break, time=0, value=1.25}
           {arriveTangent=-2.7461, interpMode=RCIM_Cubic, leaveTangent=-0.2746, tangentMode=RCTM_Break, time=0.28, value=0.05}
           <at 0.8 0>
          ]}}}
       Scale Curve=2
      )
      1=&FloatFromCurve(
       FloatCurve=DataInterface() {propertyValues {Curve {keys [
           {interpMode=RCIM_Cubic, leaveTangentWeight=0.3333, tangentMode=RCTM_Break, time=0, value=1.25}
           {arriveTangent=-2.7461, interpMode=RCIM_Cubic, leaveTangent=-0.6523, tangentMode=RCTM_Break, time=0.25, value=0.2}
           <at 0.65 0>
           <at 1 0>
          ]}}}
       Scale Curve=2
      )
      2=&FloatFromCurve(
       FloatCurve=DataInterface() {propertyValues {Curve {keys [
           {interpMode=RCIM_Cubic, leaveTangentWeight=0.3333, tangentMode=RCTM_Break, time=0, value=1.25}
           {arriveTangent=-2.7461, interpMode=RCIM_Cubic, leaveTangent=-0.6523, tangentMode=RCTM_Break, time=0.25, value=0.2}
           <at 0.65 0>
           <at 1 0>
          ]}}}
       Scale Curve=2
      )
      3=&FloatFromCurve(
       FloatCurve=DataInterface() {propertyValues {Curve {keys [
           {interpMode=RCIM_Cubic, leaveTangent=-4.0482, leaveTangentWeight=0.3333, tangentMode=RCTM_Break, time=0, value=1.25}
           {arriveTangent=-0.5108, interpMode=RCIM_Cubic, leaveTangent=-0.1916, tangentMode=RCTM_Break, time=0.4, value=0.05}
           <at 0.7 0>
          ]}}}
       Scale Curve=2
      )
      Index=$System.GameModeID
      Num=::NewEnumerator2
     ))
    ScaleA=true
    ScaleRGB=true
   )
   /SetVariables(
    Particles.SparksColor:LinearColor=&MultiplyLinearColor(
     Linear Color=<rgba 5 2.3846 0.4599 1>
     Scale Factor=&MakeVector4FromVectorAndFloat(
      "Vector (RGB)"=&VectorFromFloat(Value=&MergeHelper_FloatSwitch(0=5, 1=5, 2=5, 3=30, Index=$System.GameModeID, Num=::NewEnumerator2))
      "Float (Alpha)"=&MergeHelper_FloatSwitch(
       0=&FloatFromCurve(FloatCurve=DataInterface() {propertyValues {Curve {
           defaultValue=1
           keys [
            <at -0.0046 1.0113 RCIM_Cubic>
            {arriveTangent=-1.9677, arriveTangentWeight=0.3349, interpMode=RCIM_Cubic, leaveTangent=-1.9677, tangentMode=RCTM_User, time=1, value=0}
           ]
          }}})
       1=&FloatFromCurve(FloatCurve=DataInterface() {propertyValues {Curve {
           defaultValue=1
           keys [
            <at -0.0046 1.0113 RCIM_Cubic>
            {arriveTangent=-1.9677, arriveTangentWeight=0.3349, interpMode=RCIM_Cubic, leaveTangent=-1.9677, tangentMode=RCTM_User, time=1, value=0}
           ]
          }}})
       2=&FloatFromCurve(FloatCurve=DataInterface() {propertyValues {Curve {
           defaultValue=1
           keys [
            <at -0.0046 1.0113 RCIM_Cubic>
            {arriveTangent=-1.9677, arriveTangentWeight=0.3349, interpMode=RCIM_Cubic, leaveTangent=-1.9677, tangentMode=RCTM_User, time=1, value=0}
           ]
          }}})
       3=&FloatFromCurve(FloatCurve=DataInterface() {propertyValues {Curve {
           defaultValue=1
           keys [
            <at -0.0046 1.0113 RCIM_Cubic>
            {arriveTangent=-3.3127, arriveTangentWeight=0.3349, interpMode=RCIM_Cubic, leaveTangent=-3.3127, tangentMode=RCTM_User, time=1, value=0}
           ]
          }}})
       Index=$System.GameModeID
       Num=::NewEnumerator2
      )
     )
    )
    Particles.SparksSize:Vector2f=&Multiply_Vector2D_ByFloat(Float=&RandomRangeFloat(Minimum=4, Maximum=2), Vector2D=<60 60>)
   ) {isSetParameters=true}
   &SolveForcesAndVelocity(Clamp Velocity=false, Limit Acceleration=false)
  ]
  rendererClasses [&NiagaraSpriteRendererProperties]
  renderers [
   {rendererClass=&NiagaraSpriteRendererProperties}
   {rendererClass=&NiagaraSpriteRendererProperties, rendererIndex=1}
   {rendererClass=&NiagaraSpriteRendererProperties, rendererIndex=2}
  ]
 }}
summary {
 systemName=NS_Boost_Plumgold
 emitters [{emitterName=GoldSmoke, rendererClasses [&NiagaraSpriteRendererProperties]}]
 userVariables [
  User.bIsFrontendPreview {type=NiagaraBool}
  User.bIsPainted {type=NiagaraBool}
  User.blsIsCoroColor {type=NiagaraBool}
  User.bPreviewVehicleSpeed {type=NiagaraBool}
  User.ColorVariantID {type=NiagaraInt32}
  User.DTCarOffset {type=Vector3f, description="", defaultValue {struct=Vector3f, value=<0 0 0>}}
  User.GameModeID {type=NiagaraInt32}
  User.PaintedColor {type=LinearColor, description="", defaultValue {struct=LinearColor, value=<rgba 1 1 1 1>}}
  User.PaintedScalar {type=NiagaraFloat, defaultValue=1, description=""}
  User.PortCount {type=NiagaraInt32, defaultValue=1}
 ]
}
